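; PCG Instant graph that combines referenced or selected actors' static and skeletal meshes into one new labeled actor.
; Actor points are filtered for mesh attributes, then fed to static mesh and instanced skinned mesh spawners targeting that actor, keeping material overrides.
; An enum picks the new pivot from combined bounds, last element or world origin.
; Source actors are optionally deleted; an empty input prints an error.
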(graph Instant_Create_CombineActors
  (params (ActorReferences :type (SoftObject /Script/Engine.Actor)) (UseSelectedActors :type Bool :default true) (AssetOriginPlacement :type (Enum /PCGPrimitives/Instants/Create/Resources/EPCG_Instants_PivotPlacement.EPCG_Instants_PivotPlacement)) (ActorLabel :type String :default "CombinedActor") (DeleteSourceActors :type Bool))
  (bind Subgraph_1 (Subgraph :Graph /PCGPrimitives/Instants/Shared/Input_ActorRefsOrSelection.Input_ActorRefsOrSelection :ActorReferences ActorReferences :UseSelectedActors UseSelectedActors))
  (bind MergeAttributes_0 (Merge%20Attributes))
  (bind GetElementsCount_0 (Get%20Element%20Count))
  (bind Branch_0 (Branch))
  (bind GetActorData_4 (Get%20Actor%20Data :ActorSelector (struct :ActorFilter FromInput) :bIgnorePCGGeneratedComponents false :Mode GetPointData))
  (bind ToPoint_2 (To%20Point))
  (bind AttributeFilter_3 (Filter%20Attribute%20Elements :AttributeTypes (struct :Type SoftObjectPath) :bGenerateOutputDataEvenIfEmpty false :bUseConstantThreshold true :Operator NotEqual :TargetAttribute Mesh :ThresholdAttribute @Last))
  (bind AttributeFilter_4 (Filter%20Attribute%20Elements :AttributeTypes (struct :Type SoftObjectPath) :bGenerateOutputDataEvenIfEmpty false :bUseConstantThreshold true :Operator NotEqual :TargetAttribute SkeletalMesh :ThresholdAttribute @Last))
  (bind MergePoints_0 (Merge%20Points))
  (bind CombinePoints_0 (Combine%20Points :bUseFirstPointTransform false))
  (bind ResetPointCenter_0 (Reset%20Point%20Center :PointCenterLocation (0.5 0.5 0.0)))
  (bind AttributeReduce_0 (Attribute%20Reduce :InputSource $Index :Operation Max))
  (bind FilterElementsByIndex_0 (Filter%20Elements%20By%20Index :bSelectIndicesByInput false))
  (bind CreatePoints_0 (Create%20Points))
  (bind Select%20%28Multi%29_0 (Select%20%28Multi%29 :EnumSelection (struct :Class /PCGPrimitives/Instants/Create/Resources/EPCG_Instants_PivotPlacement.EPCG_Instants_PivotPlacement) :SelectionMode Enum :Value AssetOriginPlacement))
  (bind GetAttributeFromPointIndex_0 (Get%20Attribute%20From%20Point%20Index :bForceOutputAttributeToBeInElementsDomain true :InputSource $Position :OutputAttributeName ActorPivot))
  (bind CreateTargetActor_0 (Create%20Target%20Actor :AttachOptions NotAttached :TemplateActorClass /Script/Engine.Actor :ActorLabel-param ActorLabel))
  (bind Print%20String_0 (Print%20String :PrintString "No actor(s) with supported mesh or skeletal mesh components selected or referenced." :Verbosity Error))
  (bind AttributeFilter_0 (Filter%20Attribute%20Elements :AttributeTypes (struct :Type SoftObjectPath) :bGenerateOutputDataEvenIfEmpty false :bUseConstantThreshold true :Operator NotEqual :TargetAttribute Mesh))
  (bind StaticMeshSpawner_1 (Static%20Mesh%20Spawner :bApplyMeshBoundsToPoints false :MeshSelectorParameters (class:/Script/PCG.PCGMeshSelectorByAttribute :AttributeName Mesh :bUseAttributeMaterialOverrides true :MaterialOverrideAttributes [Material] :TemplateDescriptor (struct :bIncludeInHLOD false :bUseDefaultCollision true :bGenerateOverlapEvents true)) :MeshSelectorType /Script/PCG.PCGMeshSelectorByAttribute))
  (bind AttributeFilter_2 (Filter%20Attribute%20Elements :AttributeTypes (struct :Type SoftObjectPath) :bUseConstantThreshold true :Operator NotEqual :TargetAttribute SkeletalMesh :ThresholdAttribute @Last))
  (bind AddAttribute_1 (Add%20Attribute :AttributeTypes (struct :Type SoftObjectPath) :OutputTarget AnimBank))
  (bind AddAttribute_2 (Add%20Attribute :AttributeTypes (struct :Type Integer32 :Int32Value -1) :OutputTarget SequenceIndex))
  (bind SkinnedMeshSpawner_1 (Instanced%20Skinned%20Mesh%20Spawner :MeshSelectorParameters (class:/Script/PCG.PCGSkinnedMeshSelector :bUseAttributeMaterialOverrides true :MaterialOverrideAttributes [Material] :MeshAttribute SkeletalMesh :TemplateDescriptor (struct :bRenderStatic true :BodyInstance (struct :CollisionProfileName BlockAll)))))
  (bind Branch_1 (Branch :bOutputToB-param DeleteSourceActors))
  (bind AttributeRemoveDuplicates_0 (Attribute%20Remove%20Duplicates :AttributeSelectors [ActorReference]))
  (bind RemoveActorsFromWorld_0 (Remove%20Actors%20From%20World))
  (reroute WorldPos Select%20%28Multi%29_0)
  (connect AttributeFilter_2 InsideFilter AddAttribute_1 In)
  (connect AddAttribute_1 Out AddAttribute_2 In)
  (connect MergePoints_0 Out AttributeFilter_0 In)
  (connect MergePoints_0 Out AttributeFilter_2 In)
  (connect ToPoint_2 Out AttributeFilter_3 In)
  (connect ToPoint_2 Out AttributeFilter_4 In)
  (connect MergePoints_0 Out AttributeReduce_0 In)
  (connect Branch_1 Output%20B AttributeRemoveDuplicates_0 In)
  (connect GetElementsCount_0 Out Branch_0 bOutputToB)
  (connect MergeAttributes_0 Out Branch_0 In)
  (connect MergePoints_0 Out Branch_1 In)
  (connect MergePoints_0 Out CombinePoints_0 In)
  (connect GetAttributeFromPointIndex_0 Attribute CreateTargetActor_0 ActorPivot)
  (connect MergePoints_0 Out FilterElementsByIndex_0 In)
  (connect AttributeReduce_0 Out FilterElementsByIndex_0 SelectedIndices)
  (connect Branch_0 Output%20B GetActorData_4 In)
  (connect WorldPos Out GetAttributeFromPointIndex_0 In)
  (connect MergeAttributes_0 Out GetElementsCount_0 In)
  (connect Subgraph_1 ActorReferences MergeAttributes_0 In)
  (connect Subgraph_1 SelectedActorsReferences MergeAttributes_0 In)
  (connect AttributeFilter_3 InsideFilter MergePoints_0 In)
  (connect AttributeFilter_4 InsideFilter MergePoints_0 In)
  (connect Branch_0 Output%20A Print%20String_0 In)
  (connect AttributeRemoveDuplicates_0 Out RemoveActorsFromWorld_0 In)
  (connect CombinePoints_0 Out ResetPointCenter_0 In)
  (connect CombinePoints_0 Out Select%20%28Multi%29_0 Center)
  (connect ResetPointCenter_0 Out Select%20%28Multi%29_0 Center%20Min%20Z%20Pivot)
  (connect CombinePoints_0 Out Select%20%28Multi%29_0 Default)
  (connect FilterElementsByIndex_0 In%20Filter Select%20%28Multi%29_0 Last%20Selected%20Actor)
  (connect CreatePoints_0 Out Select%20%28Multi%29_0 World%20Origin)
  (connect AddAttribute_2 Out SkinnedMeshSpawner_1 In)
  (connect CreateTargetActor_0 Out SkinnedMeshSpawner_1 TargetActor)
  (connect AttributeFilter_0 InsideFilter StaticMeshSpawner_1 In)
  (connect CreateTargetActor_0 Out StaticMeshSpawner_1 TargetActor)
  (connect GetActorData_4 Out ToPoint_2 In))
