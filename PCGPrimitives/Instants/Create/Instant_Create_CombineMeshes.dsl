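; PCG Instant editor tool that merges static meshes from referenced or selected actors into one Nanite-enabled static mesh asset.
; Mesh points are offset by an enum-chosen pivot (center, bottom center, last actor, world origin), then appended into a dynamic mesh.
; The mesh saves via dialog or given name and path, with errors logged for missing values.
; Optional branches spawn the asset at the pivot and delete source actors, deletion not waiting on the save.
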(graph Instant_Create_CombineMeshes
  (params (ActorReferences :type (SoftObject /Script/Engine.Actor)) (UseSelectedActors :type Bool :default true) (ExportedMeshAssetName :type String) (ExportedAssetPath :type String) (OpenSaveDialog :type Bool :default true) (SpawnAssetAfterExport :type Bool) (AssetOriginPlacement :type (Enum /PCGPrimitives/Instants/Create/Resources/EPCG_Instants_PivotPlacement.EPCG_Instants_PivotPlacement)) (DeleteSourceActors :type Bool))
  (bind Subgraph_1 (Subgraph :Graph /PCGPrimitives/Instants/Shared/Input_ActorRefsOrSelection.Input_ActorRefsOrSelection :ActorReferences ActorReferences :UseSelectedActors UseSelectedActors))
  (bind MergeAttributes_0 (Merge%20Attributes))
  (bind GetElementsCount_0 (Get%20Element%20Count))
  (bind Branch_0 (Branch))
  (bind GetActorData_4 (Get%20Actor%20Data :ActorSelector (struct :ActorFilter FromInput) :bIgnorePCGGeneratedComponents false :Mode GetPointData))
  (bind ToPoint_2 (To%20Point))
  (bind MergePoints_3 (Merge%20Points))
  (bind AttributeFilter_1 (Filter%20Attribute%20Elements :AttributeTypes (struct :Type SoftObjectPath) :bUseConstantThreshold true :Operator NotEqual :TargetAttribute Mesh :ThresholdAttribute @Last))
  (bind CombinePoints_0 (Combine%20Points :bUseFirstPointTransform false))
  (bind CreatePoints_0 (Create%20Points))
  (bind AttributeReduce_0 (Attribute%20Reduce :InputSource $Index :Operation Max))
  (bind FilterElementsByIndex_0 (Filter%20Elements%20By%20Index :bSelectIndicesByInput false))
  (bind ResetPointCenter_0 (Reset%20Point%20Center :PointCenterLocation (0.5 0.5 0.0)))
  (bind Select%20%28Multi%29_0 (Select%20%28Multi%29 :EnumSelection (struct :Class /PCGPrimitives/Instants/Create/Resources/EPCG_Instants_PivotPlacement.EPCG_Instants_PivotPlacement) :SelectionMode Enum :Value AssetOriginPlacement))
  (bind AttributeMathsOp_0 (Attribute%20Maths%20Op :InputSource1 $Position :InputSource2 $Position :Operation Subtract))
  (bind CreateEmptyDynamicMesh_1 (Create%20Empty%20Dynamic%20Mesh))
  (bind AppendMeshesFromPoints_1 (Append%20Meshes%20From%20Points :MeshAttribute Mesh :Mode StaticMeshFromAttribute :RequestedLODType MaxAvailable))
  (bind Branch_1 (Branch :bOutputToB-param OpenSaveDialog))
  (bind AttributeCompareOp_6 (Attribute%20Compare%20Op :InA ExportedAssetPath))
  (bind AttributeCompareOp_5 (Attribute%20Compare%20Op :InA ExportedMeshAssetName))
  (bind Branch_3 (Branch))
  (bind Branch_5 (Branch))
  (bind Gather_0 (Gather))
  (bind SaveDynamicMeshToAsset_0 (Save%20Dynamic%20Mesh%20To%20Asset :CopyMeshToAssetOptions (struct :NewNaniteSettings (struct :bEnabled true)) :AssetName ExportedMeshAssetName :AssetPath ExportedAssetPath :bOpenSaveDialog OpenSaveDialog))
  (bind Print%20String_0 (Print%20String :bPrefixWithGraph false :bPrefixWithNode false :bPrintPerComponent false :PrintString "Missing exported mesh asset name" :Verbosity Error))
  (bind Print%20String_1 (Print%20String :bPrefixWithGraph false :bPrefixWithNode false :bPrintPerComponent false :PrintString "Missing exported asset path" :Verbosity Error))
  (bind Branch_6 (Branch :bOutputToB-param SpawnAssetAfterExport))
  (bind CopyAttributes_0 (Copy%20Attributes :InputSource AssetPath))
  (bind SpawnActor_0 (Spawn%20Actor :AttachOptions NotAttached :SpawnedActorPropertyOverrideDescriptions [(struct :InputSource AssetPath :PropertyTarget "StaticMeshComponent.StaticMesh")] :TemplateActorClass /Script/Engine.StaticMeshActor))
  (bind Branch_2 (Branch :bOutputToB-param DeleteSourceActors))
  (bind AttributeRemoveDuplicates_1 (Attribute%20Remove%20Duplicates :AttributeSelectors [ActorReference]))
  (bind RemoveActorsFromWorld_1 (Remove%20Actors%20From%20World))
  (reroute WorldPos Select%20%28Multi%29_0)
  (connect CreateEmptyDynamicMesh_1 Out AppendMeshesFromPoints_1 InDynMesh)
  (connect AttributeMathsOp_0 Out AppendMeshesFromPoints_1 InPoints)
  (connect MergePoints_3 Out AttributeFilter_1 In)
  (connect AttributeFilter_1 InsideFilter AttributeMathsOp_0 InA)
  (connect Select%20%28Multi%29_0 Out AttributeMathsOp_0 InB)
  (connect AttributeFilter_1 InsideFilter AttributeReduce_0 In)
  (connect Branch_2 Output%20B AttributeRemoveDuplicates_1 In)
  (connect GetElementsCount_0 Out Branch_0 bOutputToB)
  (connect MergeAttributes_0 Out Branch_0 In)
  (connect AppendMeshesFromPoints_1 Out Branch_1 In)
  (connect AttributeFilter_1 InsideFilter Branch_2 In)
  (connect AttributeCompareOp_5 Out Branch_3 bOutputToB)
  (connect Branch_1 Output%20A Branch_3 In)
  (connect AttributeCompareOp_6 Out Branch_5 bOutputToB)
  (connect Branch_3 Output%20A Branch_5 In)
  (connect Branch_3 Output%20B Branch_5 In)
  (connect SaveDynamicMeshToAsset_0 OutAssetPath Branch_6 In)
  (connect AttributeFilter_1 InsideFilter CombinePoints_0 In)
  (connect Branch_6 Output%20B CopyAttributes_0 Source)
  (connect WorldPos Out CopyAttributes_0 Target)
  (connect AttributeFilter_1 InsideFilter FilterElementsByIndex_0 In)
  (connect AttributeReduce_0 Out FilterElementsByIndex_0 SelectedIndices)
  (connect Branch_1 Output%20B Gather_0 In)
  (connect Branch_5 Output%20A Gather_0 In2)
  (connect Branch_0 Output%20B GetActorData_4 In)
  (connect MergeAttributes_0 Out GetElementsCount_0 In)
  (connect Subgraph_1 ActorReferences MergeAttributes_0 In)
  (connect Subgraph_1 SelectedActorsReferences MergeAttributes_0 In)
  (connect ToPoint_2 Out MergePoints_3 In)
  (connect Branch_3 Output%20B Print%20String_0 In)
  (connect Branch_5 Output%20B Print%20String_1 In)
  (connect AttributeRemoveDuplicates_1 Out RemoveActorsFromWorld_1 In)
  (connect CombinePoints_0 Out ResetPointCenter_0 In)
  (connect Gather_0 Out SaveDynamicMeshToAsset_0 In)
  (connect CombinePoints_0 Out Select%20%28Multi%29_0 Center)
  (connect ResetPointCenter_0 Out Select%20%28Multi%29_0 Center%20Min%20Z%20Pivot)
  (connect CombinePoints_0 Out Select%20%28Multi%29_0 Default)
  (connect FilterElementsByIndex_0 In%20Filter Select%20%28Multi%29_0 Last%20Selected%20Actor)
  (connect CreatePoints_0 Out Select%20%28Multi%29_0 World%20Origin)
  (connect Branch_6 Output%20B SpawnActor_0 Execution%20Dependency)
  (connect CopyAttributes_0 Out SpawnActor_0 In)
  (connect GetActorData_4 Out ToPoint_2 In))
