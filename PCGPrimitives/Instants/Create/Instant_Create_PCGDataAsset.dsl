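; PCG instant graph that exports selected or referenced actors' mesh points as a PCG Data Asset.
; Actor points with a valid Mesh are offset by a pivot chosen through an enum (center, bottom, last actor, world origin) and saved with a Root point.
; Without the save dialog, empty name or path prints errors.
; Optional branches reload and spawn the asset's meshes on a new actor at the pivot, and delete the source actors.
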(graph Instant_Create_PCGDataAsset
  (params (ActorReferences :type (SoftObject /Script/Engine.Actor)) (UseSelectedActors :type Bool :default true) (ExportedPCGDataAssetName :type String) (ExportedAssetPath :type String) (OpenSaveDialog :type Bool :default true) (AssetOriginPlacement :type (Enum /PCGPrimitives/Instants/Create/Resources/EPCG_Instants_PivotPlacement.EPCG_Instants_PivotPlacement)) (SpawnAssetAfterExport :type Bool) (DeleteSourceActors :type Bool))
  (bind Subgraph_1 (Subgraph :Graph /PCGPrimitives/Instants/Shared/Input_ActorRefsOrSelection.Input_ActorRefsOrSelection :ActorReferences ActorReferences :UseSelectedActors UseSelectedActors))
  (bind MergeAttributes_0 (Merge%20Attributes))
  (bind GetElementsCount_0 (Get%20Element%20Count))
  (bind Branch_0 (Branch))
  (bind GetActorData_4 (Get%20Actor%20Data :ActorSelector (struct :ActorFilter FromInput) :bIgnorePCGGeneratedComponents false :Mode GetPointData))
  (bind ToPoint_2 (To%20Point))
  (bind MergePoints_3 (Merge%20Points))
  (bind AttributeFilter_1 (Filter%20Attribute%20Elements :AttributeTypes (struct :Type SoftObjectPath) :bUseConstantThreshold true :Operator NotEqual :TargetAttribute Mesh :ThresholdAttribute @Last))
  (bind CombinePoints_0 (Combine%20Points :bUseFirstPointTransform false))
  (bind CreatePoints_0 (Create%20Points))
  (bind AttributeReduce_0 (Attribute%20Reduce :InputSource $Index :Operation Max))
  (bind FilterElementsByIndex_0 (Filter%20Elements%20By%20Index :bSelectIndicesByInput false))
  (bind ResetPointCenter_0 (Reset%20Point%20Center :PointCenterLocation (0.5 0.5 0.0)))
  (bind Select%20%28Multi%29_0 (Select%20%28Multi%29 :EnumSelection (struct :Class /PCGPrimitives/Instants/Create/Resources/EPCG_Instants_PivotPlacement.EPCG_Instants_PivotPlacement) :SelectionMode Enum :Value AssetOriginPlacement))
  (bind AttributeMathsOp_0 (Attribute%20Maths%20Op :InputSource1 $Position :InputSource2 $Position :Operation Subtract))
  (bind Branch_1 (Branch :bOutputToB-param OpenSaveDialog))
  (bind AttributeCompareOp_5 (Attribute%20Compare%20Op :InA ExportedPCGDataAssetName))
  (bind AttributeCompareOp_6 (Attribute%20Compare%20Op :InA ExportedAssetPath))
  (bind Branch_3 (Branch))
  (bind Print%20String_0 (Print%20String :bPrefixWithGraph false :bPrefixWithNode false :bPrintPerComponent false :PrintString "Missing exported mesh asset name" :Verbosity Error))
  (bind Branch_5 (Branch))
  (bind Print%20String_1 (Print%20String :bPrefixWithGraph false :bPrefixWithNode false :bPrintPerComponent false :PrintString "Missing exported asset path" :Verbosity Error))
  (bind Gather_0 (Gather))
  (bind PCGSaveDataAsset_0 (Save%20PCG%20Data%20Asset :AssetDescription "PCG Data Asset exported from selection using an instant operation." :Pins [(struct :Label Root :PinStatus Required) (struct :Label Points :PinStatus Required)] :AssetName ExportedPCGDataAssetName :AssetPath ExportedAssetPath :bOpenSaveDialog OpenSaveDialog))
  (bind Branch_2 (Branch :bOutputToB-param SpawnAssetAfterExport))
  (bind PCGLoadDataAsset_0 (Load%20PCG%20Data%20Asset :AssetReferenceSelector AssetPath :bLoadFromInput true))
  (bind FilterDataByTag_0 (Filter%20Data%20By%20Tag :SelectedTags "Points"))
  (bind ExtractAttributeAtIndex_0 (Extract%20Attribute%20at%20Index :InputSource $Position))
  (bind CreateTargetActor_0 (Create%20Target%20Actor :TemplateActorClass /Script/Engine.Actor))
  (bind ToPoint_0 (To%20Point))
  (bind CopyPoints_0 (Copy%20Points))
  (bind StaticMeshSpawner_0 (Static%20Mesh%20Spawner :MeshSelectorParameters (class:/Script/PCG.PCGMeshSelectorByAttribute :AttributeName Mesh :bUseAttributeMaterialOverrides true :MaterialOverrideAttributes [Material]) :MeshSelectorType /Script/PCG.PCGMeshSelectorByAttribute))
  (bind Branch_4 (Branch :bOutputToB-param DeleteSourceActors))
  (bind AttributeRemoveDuplicates_0 (Attribute%20Remove%20Duplicates :AttributeSelectors [ActorReference]))
  (bind RemoveActorsFromWorld_0 (Remove%20Actors%20From%20World))
  (reroute WorldPos Select%20%28Multi%29_0)
  (connect MergePoints_3 Out AttributeFilter_1 In)
  (connect AttributeFilter_1 InsideFilter AttributeMathsOp_0 InA)
  (connect Select%20%28Multi%29_0 Out AttributeMathsOp_0 InB)
  (connect AttributeFilter_1 InsideFilter AttributeReduce_0 In)
  (connect Branch_4 Output%20B AttributeRemoveDuplicates_0 In)
  (connect GetElementsCount_0 Out Branch_0 bOutputToB)
  (connect MergeAttributes_0 Out Branch_0 In)
  (connect AttributeMathsOp_0 Out Branch_1 In)
  (connect PCGSaveDataAsset_0 AssetPath Branch_2 In)
  (connect AttributeCompareOp_5 Out Branch_3 bOutputToB)
  (connect Branch_1 Output%20A Branch_3 In)
  (connect AttributeFilter_1 InsideFilter Branch_4 In)
  (connect AttributeCompareOp_6 Out Branch_5 bOutputToB)
  (connect Branch_3 Output%20A Branch_5 In)
  (connect Branch_3 Output%20B Branch_5 In)
  (connect AttributeFilter_1 InsideFilter CombinePoints_0 In)
  (connect ToPoint_0 Out CopyPoints_0 Source)
  (connect WorldPos Out CopyPoints_0 Target)
  (connect ExtractAttributeAtIndex_0 Out CreateTargetActor_0 ActorPivot)
  (connect Branch_2 Output%20B ExtractAttributeAtIndex_0 Execution%20Dependency)
  (connect WorldPos Out ExtractAttributeAtIndex_0 In)
  (connect PCGLoadDataAsset_0 Out FilterDataByTag_0 In)
  (connect AttributeFilter_1 InsideFilter FilterElementsByIndex_0 In)
  (connect AttributeReduce_0 Out FilterElementsByIndex_0 SelectedIndices)
  (connect Branch_1 Output%20B Gather_0 In)
  (connect Branch_5 Output%20A Gather_0 In2)
  (connect Branch_0 Output%20B GetActorData_4 In)
  (connect MergeAttributes_0 Out GetElementsCount_0 In)
  (connect Subgraph_1 ActorReferences MergeAttributes_0 In)
  (connect Subgraph_1 SelectedActorsReferences MergeAttributes_0 In)
  (connect ToPoint_2 Out MergePoints_3 In)
  (connect Branch_2 Output%20B PCGLoadDataAsset_0 In)
  (connect Gather_0 Out PCGSaveDataAsset_0 Points)
  (connect CombinePoints_0 Out PCGSaveDataAsset_0 Root)
  (connect Branch_3 Output%20B Print%20String_0 In)
  (connect Branch_5 Output%20B Print%20String_1 In)
  (connect AttributeRemoveDuplicates_0 Out RemoveActorsFromWorld_0 In)
  (connect CombinePoints_0 Out ResetPointCenter_0 In)
  (connect CombinePoints_0 Out Select%20%28Multi%29_0 Center)
  (connect ResetPointCenter_0 Out Select%20%28Multi%29_0 Center%20Min%20Z%20Pivot)
  (connect CombinePoints_0 Out Select%20%28Multi%29_0 Default)
  (connect FilterElementsByIndex_0 In%20Filter Select%20%28Multi%29_0 Last%20Selected%20Actor)
  (connect CreatePoints_0 Out Select%20%28Multi%29_0 World%20Origin)
  (connect CopyPoints_0 Out StaticMeshSpawner_0 In)
  (connect CreateTargetActor_0 Out StaticMeshSpawner_0 TargetActor)
  (connect FilterDataByTag_0 InsideFilter ToPoint_0 In)
  (connect GetActorData_4 Out ToPoint_2 In))
