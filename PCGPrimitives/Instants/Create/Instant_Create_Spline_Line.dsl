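; PCG instant graph that creates a spline component as a straight line or through multiple selected or referenced actors.
; With one or no actor, it builds a centred X-axis line of Length, placed by Transform or that actor; multiple actors supply the points.
; A Spline Sampler resamples this before final creation, with optional closed and linear settings.
; The spline goes on the single actor or a newly created one; invalid results log an error.
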(graph Instant_Create_Spline_Line
  (params (UseDistanceSamplingMode :type Bool) (SubdivisionsPerSegment :type Int) (Length :type Double :default 1e+02) (SampleDistance :type Double :default 5e+01) (FitToLength :type Bool :default true) (UseSelectedActors :type Bool :default true) (ActorReferences :type (Array (SoftObject /Script/Engine.Actor))) (ActorLabel :type String :default "SplineActor") (Transform :type Transform) (ClosedSpline :type Bool) (Linear :type Bool :default true))
  (bind GetActorData_0 (Get%20Actor%20Data :ActorSelector (struct :ActorFilter FromInput) :Mode GetSinglePoint))
  (bind GetActorData_1 (Get%20Actor%20Data :ActorSelector (struct :ActorFilter FromInput) :bMergeSinglePointData true :Mode GetSinglePoint))
  (bind AttributeTransformOp_3 (Attribute%20Transform%20Op :InputSource1 $Transform :InputSource2 @Last :InB Transform))
  (bind GetElementsCount_4 (Get%20Element%20Count))
  (bind AttributeCompareOp_4 (Attribute%20Compare%20Op))
  (bind Branch_12 (Branch))
  (bind CombinePoints_0 (Combine%20Points))
  (bind ExtractAttribute_0 (Extract%20Attribute%20at%20Index :bForceOutputAttributeToBeInElementsDomain true :InputSource $Position :OutputAttributeName ActorPivot))
  (bind CombinePoints_1 (Combine%20Points :bUseFirstPointTransform false))
  (bind ExtractAttribute_2 (Extract%20Attribute%20at%20Index :bForceOutputAttributeToBeInElementsDomain true :InputSource $Position :OutputAttributeName ActorPivot))
  ;[0.15 0.15 0.15 0.5] Input Actor References or Selection
  (bind Subgraph_0 (Subgraph :Graph /PCGPrimitives/Instants/Shared/Input_ActorRefsOrSelection.Input_ActorRefsOrSelection :ActorReferences ActorReferences :UseSelectedActors UseSelectedActors))
  ;[0 0 0 0.5] Selected actor(s) input path
  (bind GetElementsCount_1 (Get%20Element%20Count))
  (bind AttributeCompareOp_1 (Attribute%20Compare%20Op :Operation Greater))
  (bind Branch_2 (Branch))
  ;[0 0 0 0.5] Actor reference(s) input path
  (bind Branch_6 (Branch))
  (bind GetElementsCount_3 (Get%20Element%20Count))
  (bind AttributeCompareOp_3 (Attribute%20Compare%20Op :Operation Greater))
  (bind Branch_9 (Branch))
  ;[0 0 0 0.5] Creating a line when one or no actor is referenced or selected
  (bind CreatePoints_0 (Create%20Points :PointsToCreate [(struct :Transform (struct :Translation (-0.5 0.0 0.0)) :Steepness 1.0) (struct :Transform (struct :Translation (0.5 0.0 0.0)) :Steepness 1.0)]))
  (bind ApplyScaleToBounds_0 (Apply%20Scale%20To%20Bounds))
  (bind AttributeMathsOp_9 (Attribute%20Maths%20Op :Operation Max :InA Length))
  (bind AttributeMathsOp_3 (Attribute%20Maths%20Op :InputSource1 $Position.x :Operation Multiply))
  (bind AttributeTransformOp_0 (Attribute%20Transform%20Op :InputSource1 $Transform :InB Transform))
  (bind ApplyScaleToBounds_1 (Apply%20Scale%20To%20Bounds))
  (bind AttributeTransformOp_1 (Attribute%20Transform%20Op :InputSource1 $Transform :InputSource2 $Transform))
  (bind Select_2 (Select))
  (bind AttributeTransformOp_2 (Attribute%20Transform%20Op :InputSource1 $Transform :InputSource2 $Transform))
  (bind Select_4 (Select))
  ;[0 0 0 0.5] Create Initial Line
  (bind CreateSpline_0 (Create%20Spline :bLinear true))
  ;[0 0 0 0.5] Resample Initial Line/Spline for Additional Control
  (bind AttributeMathsOp_8 (Attribute%20Maths%20Op :Operation Max :InA SampleDistance))
  (bind SplineSampler_0 (Spline%20Sampler :SamplerParams (struct :bUnbounded true :PointSteepness 1.0) :bFitToCurve FitToLength :Mode UseDistanceSamplingMode :SubdivisionsPerSegment SubdivisionsPerSegment))
  ;[0 0 0 0.5] Create Spline From Multiple Referenced or Selected Actors
  (bind GetElementsCount_7 (Get%20Element%20Count))
  (bind AttributeCompareOp_7 (Attribute%20Compare%20Op :Operation Greater))
  (bind Branch_16 (Branch :In ClosedSpline))
  (bind Branch_17 (Branch :In Linear))
  (bind CreateSpline_2 (Create%20Spline :bLinear true))
  ;[0 0 0 0.5] Create Final Spline Component
  (bind GetElementsCount_6 (Get%20Element%20Count))
  (bind AttributeCompareOp_6 (Attribute%20Compare%20Op :Operation Greater))
  (bind Branch_14 (Branch :In ClosedSpline))
  (bind Branch_15 (Branch :In Linear))
  (bind CreateSpline_1 (Create%20Spline :bLinear true :Mode CreateComponent))
  ;[0 0 0 0.5] Target Actors Selection, create if no or greater than one ref or selection used, else use ref or selected actor
  (bind CreateTargetActor_0 (Create%20Target%20Actor :AttachOptions NotAttached :TemplateActorClass /Script/Engine.Actor :ActorLabel-param ActorLabel :ActorPivot-param Transform))
  (bind CreateTargetActor_1 (Create%20Target%20Actor :AttachOptions NotAttached :TemplateActorClass /Script/Engine.Actor :ActorLabel-param ActorLabel))
  (bind Select_1 (Select))
  (bind ExtractAttribute_1 (Extract%20Attribute%20at%20Index :bForceOutputAttributeToBeInElementsDomain true :InputSource ActorReference :OutputAttributeName ActorReference))
  (bind CreateTargetActor_2 (Create%20Target%20Actor :AttachOptions NotAttached :TemplateActorClass /Script/Engine.Actor :ActorLabel-param ActorLabel))
  (bind Select_3 (Select))
  ;[1 0 0 0.5] Error: Invalid Spline Created
  (bind Print%20String_1 (Print%20String :bPrefixWithNode false :bPrintPerComponent false :PrintString "Invalid Spline Created, adjust length or sampled distance." :Verbosity Error))
  (reroute HasActorRefs Subgraph_0.HasActorRefs)
  (reroute HasSelectionEnabled Subgraph_0.UseSelectedActors)
  (reroute TargetActor Select_3)
  (connect Branch_2 Output%20A ApplyScaleToBounds_0 In)
  (connect GetActorData_0 Out ApplyScaleToBounds_1 In)
  (connect GetElementsCount_1 Out AttributeCompareOp_1 InA)
  (connect GetElementsCount_3 Out AttributeCompareOp_3 InA)
  (connect GetElementsCount_4 Out AttributeCompareOp_4 InA)
  (connect GetElementsCount_6 Out AttributeCompareOp_6 InA)
  (connect GetElementsCount_7 Out AttributeCompareOp_7 InA)
  (connect CreatePoints_0 Out AttributeMathsOp_3 InA)
  (connect AttributeMathsOp_9 Out AttributeMathsOp_3 InB)
  (connect AttributeMathsOp_3 Out AttributeTransformOp_0 InA)
  (connect AttributeTransformOp_0 Out AttributeTransformOp_1 InA)
  (connect ApplyScaleToBounds_1 Out AttributeTransformOp_1 InB)
  (connect Select_2 Out AttributeTransformOp_2 InA)
  (connect ApplyScaleToBounds_0 Out AttributeTransformOp_2 InB)
  (connect Branch_2 Output%20B AttributeTransformOp_3 InA)
  (connect GetActorData_1 Out AttributeTransformOp_3 InA)
  (connect AttributeCompareOp_1 Out Branch_2 bOutputToB)
  (connect Subgraph_0 SelectedActors Branch_2 In)
  (connect HasActorRefs Out Branch_6 bOutputToB)
  (connect Subgraph_0 ActorReferences Branch_6 In)
  (connect AttributeCompareOp_3 Out Branch_9 bOutputToB)
  (connect Branch_6 Output%20B Branch_9 In)
  (connect AttributeCompareOp_4 Out Branch_12 bOutputToB)
  (connect SplineSampler_0 Out Branch_12 In)
  (connect AttributeCompareOp_6 Out Branch_14 bOutputToB)
  (connect AttributeCompareOp_6 Out Branch_15 bOutputToB)
  (connect AttributeCompareOp_7 Out Branch_16 bOutputToB)
  (connect AttributeCompareOp_7 Out Branch_17 bOutputToB)
  (connect GetActorData_1 Out CombinePoints_0 In)
  (connect Branch_2 Output%20B CombinePoints_1 In)
  (connect Branch_2 Output%20A CreatePoints_0 Execution%20Dependency)
  (connect Branch_6 Output%20A CreatePoints_0 Execution%20Dependency)
  (connect Branch_9 Output%20A CreatePoints_0 Execution%20Dependency)
  (connect Select_4 Out CreateSpline_0 In)
  (connect Branch_14 Output%20B CreateSpline_1 bClosedLoop)
  (connect Branch_15 Output%20B CreateSpline_1 bLinear)
  (connect Branch_12 Output%20A CreateSpline_1 In)
  (connect TargetActor Out CreateSpline_1 TargetActor)
  (connect Branch_16 Output%20B CreateSpline_2 bClosedLoop)
  (connect Branch_17 Output%20B CreateSpline_2 bLinear)
  (connect AttributeTransformOp_3 Out CreateSpline_2 In)
  (connect Branch_6 Output%20A CreateTargetActor_0 Execution%20Dependency)
  (connect ExtractAttribute_0 Out CreateTargetActor_1 ActorPivot)
  (connect Branch_9 Output%20B CreateTargetActor_1 Execution%20Dependency)
  (connect ExtractAttribute_2 Out CreateTargetActor_2 ActorPivot)
  (connect Branch_2 Output%20B CreateTargetActor_2 Execution%20Dependency)
  (connect CombinePoints_0 Out ExtractAttribute_0 In)
  (connect Branch_2 Output%20A ExtractAttribute_1 In)
  (connect CombinePoints_1 Out ExtractAttribute_2 In)
  (connect Branch_9 Output%20A GetActorData_0 In)
  (connect Branch_9 Output%20B GetActorData_1 In)
  (connect Subgraph_0 SelectedActors GetElementsCount_1 In)
  (connect Branch_6 Output%20B GetElementsCount_3 In)
  (connect SplineSampler_0 Out GetElementsCount_4 In)
  (connect Branch_12 Output%20A GetElementsCount_6 In)
  (connect AttributeTransformOp_3 Out GetElementsCount_7 In)
  (connect Branch_12 Output%20B Print%20String_1 In)
  (connect HasActorRefs Out Select_1 bUseInputB)
  (connect CreateTargetActor_0 Out Select_1 Input%20A)
  (connect Branch_9 Output%20A Select_1 Input%20B)
  (connect CreateTargetActor_1 Out Select_1 Input%20B)
  (connect HasActorRefs Out Select_2 bUseInputB)
  (connect AttributeTransformOp_0 Out Select_2 Input%20A)
  (connect AttributeTransformOp_1 Out Select_2 Input%20B)
  (connect HasSelectionEnabled Out Select_3 bUseInputB)
  (connect Select_1 Out Select_3 Input%20A)
  (connect CreateTargetActor_2 Out Select_3 Input%20B)
  (connect ExtractAttribute_1 Out Select_3 Input%20B)
  (connect HasSelectionEnabled Out Select_4 bUseInputB)
  (connect Select_2 Out Select_4 Input%20A)
  (connect AttributeTransformOp_2 Out Select_4 Input%20B)
  (connect AttributeMathsOp_8 Out SplineSampler_0 DistanceIncrement)
  (connect CreateSpline_0 Out SplineSampler_0 Spline)
  (connect CreateSpline_2 Out SplineSampler_0 Spline))
