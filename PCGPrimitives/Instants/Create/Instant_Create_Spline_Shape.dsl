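; PCG instant graph that creates an N-gon spline shape and spawns it as a spline actor.
; It exposes width, length, side count, linear toggle, position, rotation and world-origin placement, passing them straight to the Create_Spline_Shape subgraph.
; That subgraph's output feeds the Spawn_SplineActor subgraph, so this graph is a thin wrapper chaining two reusable primitives.
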(graph Instant_Create_Spline_Shape
  (params (Width :type Double :default 5e+03) (Length :type Double :default 5e+03) (Ngon :type Int :default 12) (Linear :type Bool) (Position :type Vector) (Rotation :type Rotator) (AtWorldOrigin :type Bool :default true))
  (bind CreateSplineShape_0 (Subgraph :Graph /PCGPrimitives/Primitives/Create/Create_Spline_Shape.Create_Spline_Shape :AtWorldOrigin AtWorldOrigin :Length Length :Linear Linear :Ngon Ngon :Position Position :Rotation Rotation :Width Width))
  (bind SpawnSplineActor_0 (Subgraph :Graph /PCGPrimitives/Primitives/Spawn/Spawn_SplineActor.Spawn_SplineActor))
  (connect CreateSplineShape_0 Out SpawnSplineActor_0 PrimaryInput))
