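; PCG graph that creates a spline component on a new, unattached plain Actor from an array of control-point transforms.
; It converts the transforms to points, which feed Create Spline directly, with parameters for closed loop and linear (linear by default).
; The same points are also combined without using the first point's transform, and the merged point's position becomes the new actor's pivot.
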(graph Instant_Create_Spline_ShapeFromPoints
  (params (ControlPoints :type (Array Transform)) (ClosedLoop :type Bool) (Linear :type Bool :default true))
  (bind AttributeSetToPoint_0 (Attribute%20Set%20To%20Point :AttributeToConvert (map (ControlPoints $Transform)) :In ControlPoints))
  (bind CombinePoints_0 (Combine%20Points :bUseFirstPointTransform false))
  (bind GetAttributeFromPointIndex_0 (Get%20Attribute%20From%20Point%20Index :bForceOutputAttributeToBeInElementsDomain true :InputSource $Position))
  (bind CreateTargetActor_1 (Create%20Target%20Actor :AttachOptions NotAttached :TemplateActorClass /Script/Engine.Actor))
  (bind CreateSpline_0 (Create%20Spline :Mode CreateComponent :bClosedLoop-param ClosedLoop :bLinear-param Linear))
  (connect AttributeSetToPoint_0 Out CombinePoints_0 In)
  (connect AttributeSetToPoint_0 Out CreateSpline_0 In)
  (connect CreateTargetActor_1 Out CreateSpline_0 TargetActor)
  (connect GetAttributeFromPointIndex_0 Attribute CreateTargetActor_1 ActorPivot)
  (connect CombinePoints_0 Out GetAttributeFromPointIndex_0 In))
